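# Light function material for the saucer vehicle emote that breaks light up into an animated, caustic-like turbulence pattern.
# Two turbulence samples with separate UV scales and speeds are panned or rotated by static switch and distorted by cloud noise mapped on world XY.
# Their doubled product is square-rooted to blend between Low and High intensity parameters; mobile shading outputs constant full brightness.
# It is flagged compatible with the light function atlas.
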Material /BRCosmetics/Animation/Props/Saucer_Vehicle/FX/M_Saucer_Emote_LightFunction
graph {
 properties {
  bForceCompatibleWithLightFunctionAtlas=true
  materialDomain=MD_LightFunction
  pixelDepthOffsetMode=PDOM_Legacy
  shadingModel=MSM_Unlit
  usedShadingModels=Unlit
 }
 nodes [
  n1 = ScalarParameter() {defaultValue=0.01, parameterName=Low}
  n2 = ScalarParameter() {defaultValue=1, parameterName=High}
  n3 = WorldPosition()
  n4 = TextureSample(UVs=Divide(A=ComponentMask(None=@n3.XYZ) {g=true, r=true}) {constB=4096}) {samplerType=SAMPLERTYPE_LinearColor, texture=/Game/Missions/Primary/DeliverTheBomb/Textures/T_CloudNoise}
  n5 = Multiply(A=Subtract(A=AppendVector(A=@n4.R, B=@n4.R)) {constB=0.5}) {constB=0.2}
  n6 = Add(A=@n5, B=Multiply(A=TextureCoordinate(), B=ScalarParameter() {defaultValue=4, parameterName=Scale UV A}))
  n7 = Time()
  n8 = Multiply(A=ScalarParameter() {defaultValue=0.25, parameterName=Scale Time A}, B=@n7)
  n9 = TextureSample(UVs=StaticSwitchParameter(True=Panner(Coordinate=@n6, Time=@n8) {speedX=1, speedY=1}, False=Rotator(Coordinate=@n6, Time=@n8) {centerX=2, centerY=2, speed=1}) {desc="T=Panner; F=Rotator", parameterName=Panner/Rotator}) {texture=/Game/Global/Textures/Default/Caustics/T_Turbulence_Seamless}
  n10 = Add(A=Multiply(A=TextureCoordinate(), B=ScalarParameter() {defaultValue=8, parameterName=Scale UV B}), B=@n5)
  n11 = Multiply(A=@n7, B=ScalarParameter() {defaultValue=-0.1, parameterName=Scale Time B})
  n12 = TextureSample(UVs=StaticSwitchParameter(True=Panner(Coordinate=@n10, Time=@n11) {speedX=1, speedY=1}, False=Rotator(Coordinate=@n10, Time=@n11) {centerX=4, centerY=4, speed=1}) {desc="T=Panner; F=Rotator", parameterName=Panner/Rotator}) {texture=&T_Turbulence_Seamless}
  n13 = SquareRoot(None=Multiply(A=Multiply(A=@n9.R, B=@n12.R)) {constB=2})
  n14 = Constant() {r=1}
  ShadingPathSwitch(Default=ShadingPathSwitch(Default=LinearInterpolate(A=@n1, B=@n2, Alpha=@n13) {constA=0.01}, Mobile=@n14), Mobile=@n14) {desc=Migrated from FeatureLevelSwitch}
 ]
}
