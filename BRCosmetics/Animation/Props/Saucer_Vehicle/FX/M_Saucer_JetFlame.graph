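# Additive unlit material for the saucer vehicle's jet flame, used on Niagara mesh particles.
# Layered Fresnel exponents shape outer, center and inner glow, faded along UV V for opacity, with depth fade and optional particle alpha.
# Emissive adds a particle-tinted flame color to a root color masked by a powered V gradient, through MF_FX_EmissiveAdjustments.
# It also derives refraction from the opacity mask, scaled by particle red and fading with camera distance.
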Material /BRCosmetics/Animation/Props/Saucer_Vehicle/FX/M_Saucer_JetFlame
graph {
 properties {
  bUsedWithNiagaraMeshParticles=true
  blendMode=BLEND_Additive
  pixelDepthOffsetMode=PDOM_Legacy
  refractionMethod=RM_IndexOfRefraction
  shadingModel=MSM_Unlit
  usedShadingModels=Unlit
 }
 nodes [
  n1 = OneMinus(None=Fresnel(ExponentIn=ScalarParameter() {defaultValue=0.4, parameterName=OuterFresnelExponent}) {baseReflectFraction=0})
  n2 = OneMinus(None=Power(Base=@n1, Exp=ScalarParameter() {defaultValue=4, parameterName=CenterFresnelExponent}))
  n3 = ScalarParameter() {defaultValue=0.4, parameterName=CenterFresnelInvertExponent}
  n4 = Max(A=Power(Base=@n1, Exp=ScalarParameter() {defaultValue=15, parameterName=InnerFresnelExponent}), B=OneMinus(None=Power(Base=@n2, Exp=@n3)))
  n5 = MaterialFunctionCall(Float2=TextureCoordinate()) {materialFunction=&BreakOutFloat2Components}
  n6 = ScalarParameter() {parameterName=VFadeOffset}
  n7 = Multiply(A=@n4, B=Power(Base=Add(A=@n5.G, B=@n6), Exp=ScalarParameter() {defaultValue=5, parameterName=VFadePower}))
  n8 = Power(Base=OneMinus(None=@n7), Exp=ScalarParameter() {parameterName=RefractionFromOpacityPower})
  n9 = ParticleColor()
  n10 = MaterialFunctionCall(Vector 3=Subtract(A=WorldPosition().XYZ, B=CameraPositionWS())) {materialFunction=/Engine/Functions/Engine_MaterialFunctions02/Utility/VectorLength}
  n11 = ScalarParameter() {defaultValue=3000, parameterName=RefFadeDistance}
  n12 = OneMinus(None=Saturate(None=Divide(A=@n10.V3 Length, B=@n11)))
  n13 = Constant() {r=1}
  LinearInterpolate(
   B=LinearInterpolate(B=ScalarParameter() {parameterName=RefractionStrength}, Alpha=OneMinus(None=@n8)) {constA=1}
   Alpha=Multiply(A=@n9.R, B=Multiply(A=@n12, B=@n13))
  ) {constA=1}
  n14 = VectorParameter() {parameterName=OutFlameColor, defaultValue=<rgba 0 0.2833 1 1>}
  n15 = VectorParameter() {parameterName=RootFlameColor, defaultValue=<rgba 0 2 1.9069 1>}
  n16 = Add(A=@n5.G, B=ScalarParameter() {defaultValue=0.01, parameterName=VRootOffset})
  n17 = Clamp(None=Power(Base=Saturate(None=@n16), Exp=ScalarParameter() {defaultValue=30, parameterName=VRootPower})) {clampMode=CMODE_ClampMin}
  MaterialFunctionCall(Emissive=Add(A=Multiply(A=@n14.RGB, B=@n9.RGB), B=Multiply(A=@n15.RGB, B=@n17))) {materialFunction=&MF_FX_EmissiveAdjustments}
  n18 = Multiply(A=@n7, B=ScalarParameter() {defaultValue=1, parameterName=OpacityMul})
  n19 = Multiply(A=Saturate(None=@n4), B=Multiply(A=Saturate(None=@n17), B=ScalarParameter() {parameterName=RootOpacityMul}))
  Multiply(
   A=DepthFade(Opacity=Add(A=@n18, B=@n19), FadeDistance=ScalarParameter() {parameterName=DepthFade})
   B=LinearInterpolate(B=@n9.A, Alpha=ScalarParameter() {parameterName=UseParticleAlpha?}) {constA=1}
  )
 ]
}
