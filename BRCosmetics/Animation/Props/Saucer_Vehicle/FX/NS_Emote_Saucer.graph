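# Niagara effect for the saucer vehicle emote, producing smoke, bouncing rocks, rings and streaks, apparently as hover downwash on the ground.
# System scripts supply a smoke center position that anchors the effect and a HeightFactor that scales spawn rates and smoke strength.
# User flags shut off rocks, rings and streaks in water and adjust the front-end preview look.
# Pulsed smoke varies with time of day and camera distance; the Minimal emitter is disabled.
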NiagaraSystem /BRCosmetics/Animation/Props/Saucer_Vehicle/FX/NS_Emote_Saucer
dependencies {
 usedDataInterfaces [&NiagaraDataInterfaceCurve, &NiagaraDataInterfaceCamera, &NiagaraDataInterfaceColorCurve]
 usedDynamicInputs [&ClampFloat, &Divide_Float, &Subtract_Float]
 usedModules [
  &SystemState
  &self:CalcSmokeCenterPosition
  &self:SystemSpawnScript.SystemScriptSource.SystemScriptGraph.NiagaraNodeAssignment_0.SetVariables_#1
  &self:SystemSpawnScript.SystemScriptSource.SystemScriptGraph.NiagaraNodeAssignment_1.SetVariables_#2
 ]
}
emitters {
 GroundRocks {
  emitterUpdateScript [
   &EmitterState()
   &SpawnRate(
    SpawnRate=&Lerp_Float(
     A=0
     B=&RandomRangeFloat(Minimum=-2, Maximum=&Lerp_Float(A=-23.625, B=33.75, Alpha=$System.HeightFactor), Evaluation Type=::NewEnumerator1)
     Alpha=&MakeCustomFloatFromBool(Bool=&Boolean_OrOperation(Bool1=$User.bIsInWater, Bool2=$User.bIsFrontEndPreview), True Float=0, False Float=1)
    )
    Use Spawn Probability=false
   )
  ]
  particleSpawnScript [
   /SetVariables(Particles.MaterialRandom:NiagaraFloat=&RandomRangeFloat(Minimum=0, Maximum=1)) {isSetParameters=true}
   /Niagara/Modules/Spawn/Initialization/InitializeParticle(
    Color=&LerpLinearColors(StartColor=<rgba 0.65 0.5738 0.5193 1>, EndColor=<rgba 0.75 0.6437 0.5677 1>, LerpFactor=$Particles.MaterialRandom)
    Lifetime=&Lerp_Float(A=0.7, B=1, Alpha=$Particles.MaterialRandom)
    Module.Write Color=true
    Module.Write Lifetime=true
    Module.Write Mass=true
    Module.Write Material Random=false
    Module.Write Position=false
    Module.Write Scale=false
    Module.Write SpriteRotation=true
    Module.Write SpriteSize=true
    Position=&SimulationPosition()
    Sprite Rotation=&RandomRangeFloat(Minimum=0, Maximum=360)
    Sprite Size=&Vector2DFromFloat(Value=&RandomRangeFloat(Minimum=3, Maximum=7))
   )
   /SetVariables(Particles.Position:Vector3f=$System.SmokeCenterPositionWS) {isSetParameters=true}
   &SphereLocation(
    MaskSpawnGroup=false
    Non Uniform Scale=<1 1 0>
    Offset=<0 0 5>
    Override Seed=false
    Sphere Origin:Vector3f=$System.SmokeCenterPositionWS
    Sphere Radius=&Lerp_Float(A=25, B=25, Alpha=&Power(a=$Particles.MaterialRandom, b=1.5))
    Surface Only=false
    UseNonUniformScale=true
    UseOffset=true
    UseSphereOrientation=false
   )
   &AddVelocityFromPoint(
    Module.Use Falloff Distance=false
    Velocity Origin:Vector3f=$System.SmokeCenterPositionWS
    Velocity Strength=&Lerp_Float(A=180, B=220, Alpha=&Power(a=$Particles.MaterialRandom, b=3))
   )
   &AddVelocity(Velocity=<0 0 200>)
  ]
  particleUpdateScript [
   &ParticleState()
   &GravityForce()
   &Drag(Use Linear Drag=true, Use Rotational Drag=false)
   &SpriteRotationRate(Rotation Rate=&Lerp_Float(A=100, B=650, Alpha=$Particles.MaterialRandom))
   &ScaleSpriteSize(Scale Factor=&Vector2DFromFloat(Value=&FloatFromCurve(FloatCurve=DataInterface() {propertyValues {Curve {keys [
          {arriveTangent=35.6763, interpMode=RCIM_Cubic, leaveTangent=3.7835, tangentMode=RCTM_User, time=0, value=0}
          <at 0.7 1 RCIM_Cubic>
          {interpMode=RCIM_Cubic, leaveTangent=0.6903, time=1, value=0}
         ]}}})))
   &Collision(
    Advanced Aging Rate=1
    Age Colliding Particles=false
    Collision Enabled=&SetBoolByFloatComparison(A=$Particles.Age, B=0.1)
    Module.Control Rest State Via Surface Normal=false
    Module.EnableMaxCollisionCount=false
    Module.Randomize Collision Normal Vector=false
    Restitution=0.3
   )
   /SetVariables(Particles.MaxCollisions:NiagaraBool=&SetBoolByIntComparison(A=$Particles.Number Of Collisions, B=1, Comparison Type=::NewEnumerator0)) {isSetParameters=true}
   &SolveForcesAndVelocity(Clamp Velocity=false, Limit Acceleration=false)
   /SetVariables(Particles.SpriteRotation:NiagaraFloat=&MakeCustomFloatFromBool(Bool=$Particles.MaxCollisions, True Float=0, False Float=$Particles.SpriteRotation)) {isSetParameters=true}
   &ScaleColor(
    Scale RGB=&VectorFromFloat(Value=&MakeCustomFloatFromBool(Bool=$User.bIsFrontEndPreview, True Float=1.5, False Float=1))
    ScaleA=true
    ScaleRGB=true
   )
  ]
  rendererClasses [&NiagaraSpriteRendererProperties]
  renderers [{rendererClass=&NiagaraSpriteRendererProperties}]
 }
 Minimal {
  bEnabled=false
  emitterUpdateScript [
   &EmitterState()
   &SpawnBurst_Instantaneous(Use Spawn Probability=false, UseLoopCountLimit=false)
  ]
  particleSpawnScript [
   /Niagara/Modules/Spawn/Initialization/V2/InitializeParticle(Color Mode=::NewEnumerator0, Lifetime=5, UsePositionOffset=false, Write Material Random=true)
   &SystemLocation(MaskSpawnGroup=false, Offset=$System.Test)
  ]
  particleUpdateScript [&ParticleState()]
  rendererClasses [&NiagaraSpriteRendererProperties, &NiagaraMeshRendererProperties]
  renderers [
   {rendererClass=&NiagaraSpriteRendererProperties}
   {rendererClass=&NiagaraMeshRendererProperties, rendererIndex=1}
  ]
 }
 SmokeGroundRings {
  emitterUpdateScript [
   &EmitterState()
   &SpawnRate(
    SpawnRate=&Lerp_Float(
     A=0
     B=&FloatFromCurve(FloatCurve=DataInterface() {propertyValues {Curve {keys [<at 0 0>,<at 0.6 0>,<at 1 10>]}}}, CurveIndex=$System.HeightFactor)
     Alpha=&MakeCustomFloatFromBool(Bool=$User.bIsInWater, True Float=0, False Float=1)
    )
    Use Spawn Probability=false
   )
  ]
  particleSpawnScript [
   /Niagara/Modules/Spawn/Initialization/InitializeParticle(
    Color=<rgba 0.635 0.8149 0.919 0.3>
    Lifetime=&RandomRangeFloat(Minimum=0.4, Maximum=0.6)
    Module.Write Color=true
    Module.Write Lifetime=true
    Module.Write Mass=true
    Module.Write Material Random=true
    Module.Write Position=false
    Module.Write Scale=false
    Module.Write SpriteRotation=true
    Module.Write SpriteSize=true
    Position=&SimulationPosition()
    Sprite Rotation=&UniformRangedFloat(Minimum=-180, Maximum=180)
    Sprite Size=&Vector2DFromFloat(Value=6.5)
   )
   /SetVariables(Particles.SpriteFacing=<0 0 1>) {isSetParameters=true}
  ]
  particleUpdateScript [
   &ParticleState()
   &DynamicMaterialParameters(
    Index 0 Param 1=&FloatFromCurve(FloatCurve=DataInterface() {propertyValues {Curve {keys [<at 0 0.5>,<at 1 2>]}}})
    Module.Param0WriteEnabled=true
    Module.Param1WriteEnabled=false
    Module.Param2WriteEnabled=false
    Module.Param3WriteEnabled=false
   )
   &ScaleSpriteSize(Scale Factor=&Vector2DFromFloat(Value=&FloatFromCurve(
      FloatCurve=DataInterface() {propertyValues {Curve {keys [
          {arriveTangent=35.6763, interpMode=RCIM_Cubic, leaveTangent=3.7835, tangentMode=RCTM_User, time=0, value=0}
          {interpMode=RCIM_Cubic, leaveTangent=0.6903, time=1.0011, value=1}
         ]}}}
      Scale Curve=50
     )))
   &CameraOffset(Camera Offset Amount=20)
   &ScaleColor(
    Scale Alpha=&FloatFromCurve(FloatCurve=DataInterface() {propertyValues {Curve {keys [<at 0 1 RCIM_Cubic>, {interpMode=RCIM_Cubic, leaveTangent=-4.9981, time=1.0011, value=0.002}]}}})
    ScaleA=true
    ScaleRGB=false
   )
   /SetVariables(Particles.Position:Vector3f=$System.SmokeCenterPositionWS) {isSetParameters=true}
  ]
  rendererClasses [&NiagaraSpriteRendererProperties]
  renderers [{rendererClass=&NiagaraSpriteRendererProperties}]
 }
 Streaks001 {
  emitterUpdateScript [
   &EmitterState()
   &SpawnRate(
    SpawnRate=&Lerp_Float(
     A=0
     B=&Lerp_Float(A=-50, B=100, Alpha=$System.HeightFactor)
     Alpha=&MakeCustomFloatFromBool(Bool=$User.bIsInWater, True Float=0, False Float=1)
    )
    Use Spawn Probability=false
   )
  ]
  particleSpawnScript [
   &InitializeParticle(
    Color=&MakeLinearColorFromVectorAndFloat("Vector (RGB)"=&MakeVectorFromLinearColorRGB(Color=<rgba 0.629 0.7403 1 1>), "Float (Alpha)"=0.1)
    Lifetime=&UniformRangedFloat(Minimum=0.3, Maximum=0.5)
    Module.Write Color=true
    Module.Write Lifetime=true
    Module.Write Mass=true
    Module.Write Material Random=true
    Module.Write Position=false
    Module.Write Scale=false
    Module.Write SpriteRotation=true
    Module.Write SpriteSize=true
    Position=&SimulationPosition()
    Sprite Size=&Multiply_Vector2(
     A=&UniformRangedVector2D(Minimum=<20 116.296>, Maximum=<40 209.1042>)
     B=&Vector2DFromFloat(Value=0.3)
    )
   )
   &CylinderLocation(
    Cylinder Height=0
    Cylinder Radius=1
    MaskSpawnGroup=false
    Module.Enable Lathe Profile=false
    Module.Override Randomness=false
    Module.Surface Only=true
    Surface Only Band Thickness=1
    Use Endcaps In Surface Only Mode=false
   )
   &AddVelocity(Velocity=&MakeVector(X=0, Y=0, Z=&UniformRangedFloat(Minimum=50, Maximum=100)))
   &AddVelocityFromPoint(
    Module.Use Falloff Distance=false
    Velocity Origin:Vector3f=&SimulationPosition()
    Velocity Strength=&Lerp_Float(A=250, B=450, Alpha=$System.HeightFactor)
   )
   &ApplyInitialForces()
   &DynamicMaterialParameters(
    Index 0 Param 1=&UniformRangedFloat(Minimum=0, Maximum=1)
    Index 0 Param 2=&UniformRangedFloat(Minimum=1, Maximum=2)
    Index 0 Param 3=1
    Index 0 Param 4=1
    Module.Param0WriteEnabled=true
    Module.Param1WriteEnabled=true
    Module.Param2WriteEnabled=true
    Module.Param3WriteEnabled=true
   )
  ]
  particleUpdateScript [
   &ParticleState()
   &Drag(Drag=1.5, Use Linear Drag=true, Use Rotational Drag=false)
   &SolveForcesAndVelocity(Clamp Velocity=false, Limit Acceleration=false)
   &ScaleColor(
    Scale Alpha=&FloatFromCurve(
     FloatCurve=DataInterface() {propertyValues {Curve {keys [<at 0 0 RCIM_Cubic>, {arriveTangent=0.002, interpMode=RCIM_Cubic, leaveTangent=0.002, time=0.5, value=1}, <at 1.0011 0.002 RCIM_Cubic>]}}}
     Scale Curve=0.8
    )
    Scale RGB=&MakeVectorFromLinearColorRGB(Color=&ColorFromCurve(ColorCurve=DataInterface() {propertyValues {
        AlphaCurve {keys [<at 0 1 RCIM_Cubic>,<at 1 0 RCIM_Cubic>]}
        BlueCurve {keys [<at 0 1.14 RCIM_Cubic>,<at 1 1 RCIM_Cubic>]}
        GreenCurve {keys [<at 0 1.842 RCIM_Cubic>,<at 1 1 RCIM_Cubic>]}
        RedCurve {keys [<at 0 3 RCIM_Cubic>,<at 1 1 RCIM_Cubic>]}
       }}))
    ScaleA=true
    ScaleRGB=true
   )
   /SetVariables(Particles.Pos2:Vector3f=&Add_Vector(
     A=Linked() {linkedVariable {name=Particles.Position, type=NiagaraPosition}}
     B=&Subtract_Vector(A=$System.SmokeCenterPositionWS, B=Linked() {linkedVariable {name=Engine.Owner.Position, type=NiagaraPosition}})
    )) {isSetParameters=true}
  ]
  rendererClasses [&NiagaraSpriteRendererProperties]
  renderers [{rendererClass=&NiagaraSpriteRendererProperties}]
 }
 smoke_pulse {
  emitterUpdateScript [
   &EmitterState(Life Cycle Mode=::NewEnumerator1, Loop Duration=0.2, UseLoopDelay=false)
   &SpawnBurst_Instantaneous(
    Spawn Count=&MakeIntFromFloat(Float=&Lerp_Float(A=25, B=40, Alpha=$System.HeightFactor))
    Use Spawn Probability=false
   )
   &SpawnRate(SpawnGroup=1, SpawnRate=&Lerp_Float(A=15, B=70, Alpha=$System.HeightFactor), Use Spawn Probability=false)
  ]
  particleSpawnScript [
   /Niagara/Modules/Spawn/Initialization/V2/InitializeParticle(
    Color=&MakeLinearColorFromVectorAndFloat(
     "Vector (RGB)"=&Multiply_Vector(
      A=&MakeCustomVectorFromBool(
       Bool=$User.bIsFrontEndPreview
       True Vector=&MakeVectorFromLinearColorRGB(Color=<rgba 0.4 0.4 0.4 1>)
       False Vector=&MakeVectorFromLinearColorRGB(Color=<rgba 0.13 0.13 0.13 1>)
      )
      B=&VectorFromFloat(Value=1)
     )
     "Float (Alpha)"=&Multiply_Float(
      A=&MakeCustomFloatFromBool(
       Bool=$User.bIsFrontEndPreview
       True Float=0.03
       False Float=&FN_Select_Float_from_Time_Of_Day(Use Manual Time of Day Thresholds=true, Morning=0.12, Day=0.12, Evening=0.09, Night=0.17, Evening Threshold=20)
      )
      B=&Lerp_Float(A=0.6, B=1.2, Alpha=$System.HeightFactor)
     )
    )
    Lifetime=&RandomRangeFloat(Minimum=0.6, Maximum=0.7)
    Sprite UV Mode=::NewEnumerator3
    UsePositionOffset=false
    Write Material Random=true
   )
   &CylinderLocation(
    Cylinder Height=0
    Cylinder Origin:Vector3f=$System.SmokeCenterPositionWS
    Cylinder Radius=30
    MaskSpawnGroup=false
    Module.Enable Lathe Profile=false
    Module.Override Randomness=false
    Module.Surface Only=true
    Surface Only Band Thickness=0
   )
   &AddVelocityFromPoint(
    Module.Use Falloff Distance=false
    Velocity Origin:Vector3f=$System.SmokeCenterPositionWS
    Velocity Strength=&Multiply_Float(
     A=&Lerp_Float(A=320, B=420, Alpha=$Particles.MaterialRandom)
     B=&Lerp_Float(A=0.8, B=1.5, Alpha=$System.HeightFactor)
    )
   )
   /SetVariables(Particles.SpriteSize=&Vector2DFromFloat(Value=&MaskFloatBySpawnGroup(
      Masked Float=&RandomRangeFloat(Minimum=90, Maximum=120)
      Passthrough Float=&RandomRangeFloat(Minimum=110, Maximum=140)
     ))) {isSetParameters=true}
  ]
  particleUpdateScript [
   &ParticleState()
   &LinearForce(
    Coordinate Space:ENiagaraCoordinateSpace=::World
    Force=&Lerp_Vector(A=<0 0 0>, B=<0 0 110>, Alpha=$System.HeightFactor)
   )
   &Drag(Drag=7, Use Linear Drag=true, Use Rotational Drag=false)
   &SolveForcesAndVelocity(Clamp Velocity=false, Limit Acceleration=false)
   &ScaleColor(
    Scale Alpha=&FloatFromCurve(FloatCurve=DataInterface() {propertyValues {Curve {keys [
         <at 0 0 RCIM_Cubic>
         {arriveTangent=-1.9555, interpMode=RCIM_Cubic, leaveTangent=-1.9555, tangentMode=RCTM_User, time=0.2374, value=1}
         <at 1 0>
        ]}}})
    ScaleA=true
    ScaleRGB=true
   )
   &ScaleSpriteSize(Scale Factor=&Vector2DFromFloat(Value=&FloatFromCurve(FloatCurve=DataInterface() {propertyValues {Curve {keys [<at 0 0.5 RCIM_Cubic>, {arriveTangent=0.5532, interpMode=RCIM_Cubic, leaveTangent=-1.4995, tangentMode=RCTM_User, time=1, value=1}]}}})))
   &DynamicMaterialParameters(
    Index 0 Param 1=&FloatFromCurve(FloatCurve=DataInterface() {propertyValues {Curve {keys [<at 0 1>,<at 1 0>]}}}, Scale Curve=0.8)
    Index 0 Param 2=&FN_Select_Float_from_Time_Of_Day(Use Manual Time of Day Thresholds=true, Morning=1, Day=1, Evening=1, Night=0.2, Evening Threshold=20)
    Module.Param0WriteEnabled=true
    Module.Param1WriteEnabled=true
    Module.Param2WriteEnabled=true
    Module.Param3WriteEnabled=true
   )
   &SubUVAnimation(SubUV Loop Count=0)
   &self:CustomCameraOffset(
    Curve=DataInterface() {propertyValues {Curve {keys [
        {tangentMode=RCTM_User, time=0, value=0}
        {interpMode=RCIM_Cubic, leaveTangent=6.6844, tangentMode=RCTM_Break, time=0.4, value=0}
        {tangentMode=RCTM_User, time=0.8, value=1}
       ]}}}
    OffsetValue:Vector3f=<0 0 8>
   ) {module=CustomCameraOffset}
   &self:CameraDistanceFade(
    MaxDistance=2000
    New Curve for Floats=DataInterface() {propertyValues {Curve {keys [
        <at 0 1>
        {interpMode=RCIM_Cubic, leaveTangent=0.2164, tangentMode=RCTM_Break, time=0.1, value=1}
        {arriveTangent=-0.194, interpMode=RCIM_Cubic, tangentMode=RCTM_Break, time=0.5, value=0.4}
       ]}}}
   ) {module=CameraDistanceFade}
   &ScaleColor(Scale Alpha=$Particles.CameraDistanceAlpha, ScaleA=true, ScaleRGB=true) {module=ScaleColor001}
  ]
  rendererClasses [&NiagaraSpriteRendererProperties]
  renderers [
   {rendererClass=&NiagaraSpriteRendererProperties}
   {rendererClass=&NiagaraSpriteRendererProperties, rendererIndex=1}
  ]
 }
}
summary {
 systemName=NS_Emote_Saucer
 emitters [
  {emitterName=GroundRocks, rendererClasses [&NiagaraSpriteRendererProperties]}
  {emitterName=smoke_pulse, rendererClasses [&NiagaraSpriteRendererProperties]}
  {emitterName=SmokeGroundRings, rendererClasses [&NiagaraSpriteRendererProperties]}
  {emitterName=Streaks001, rendererClasses [&NiagaraSpriteRendererProperties]}
  {
   bEnabled=false
   emitterName=Minimal
   rendererClasses [&NiagaraSpriteRendererProperties, &NiagaraMeshRendererProperties]
  }
 ]
 userVariables [
  User.bIsFrontEndPreview {type=NiagaraBool}
  User.bIsInWater {type=NiagaraBool}
  User.InitDelay {type=NiagaraFloat, defaultValue=0, description=""}
  User.PlayerSpeed {type=NiagaraFloat, defaultValue=0, description=""}
  User.SkeletalMesh {
   type=NiagaraDataInterfaceSkeletalMesh
   defaultValue {
    struct=NiagaraExt_VariableValue_DataInterface
    value {dataInterface=&self:NiagaraDataInterfaceSkeletalMesh_0}
   }
  }
 ]
}
