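# Niagara effect for a saucer vehicle emote prop, combining downward thruster jets, falling sparkle dust and a spotlight.
# Three jet emitters each burst one long-lived particle with sprite and mesh renderers, randomizing scale and brightness for flicker.
# User.bIsFrontEndPreview sets an integer tag on every jet and lowers Jet001's alpha in preview.
# Sparkle dust spawns continuously on a sphere with a pulsing alpha curve, while a component renderer supplies the warm-tinted spotlight.
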NiagaraSystem /BRCosmetics/Animation/Props/Saucer_Vehicle/FX/NS_Emote_SaucerPart
dependencies {
 usedDataInterfaces [&NiagaraDataInterfaceCurve, &NiagaraDataInterfaceColorCurve]
 usedDynamicInputs []
 usedModules [&SystemState, &self:CalcSmokeCenterPosition]
}
emitters {
 FN_SparkleDust_Constant {
  emitterUpdateScript [
   &EmitterState(Inactive Response=::NewEnumerator1, Life Cycle Mode=::NewEnumerator1, Loop Duration=1, UseLoopDelay=false)
   &SpawnRate(SpawnRate=7.4442, Use Spawn Probability=false)
  ]
  particleSpawnScript [
   /Niagara/Modules/Spawn/Initialization/InitializeParticle(
    Lifetime=&UniformRangedFloat(Minimum=0.9807, Maximum=1.1656)
    Mesh Scale=&Multiply_Vector(
     A=<1 1 1>
     B=&VectorFromFloat(Value=&UniformRangedFloat(Minimum=0.2, Maximum=1))
    )
    Module.Write Color=false
    Module.Write Lifetime=true
    Module.Write Mass=true
    Module.Write Material Random=true
    Module.Write Position=false
    Module.Write Scale=false
    Module.Write SpriteRotation=true
    Module.Write SpriteSize=true
    Position=&SimulationPosition()
    Sprite Rotation=&UniformAOrBFloat(A=0, B=360)
    Sprite Size=&Vector2DFromFloat(Value=&UniformRangedFloat(Minimum=1, Maximum=6))
   )
   &SphereLocation(
    MaskSpawnGroup=false
    Offset=<0 0 -20>
    Override Seed=false
    Sphere Radius=20.4819
    Surface Only=true
    UseNonUniformScale=false
    UseOffset=true
    UseSphereOrientation=false
   )
   &InheritVelocity(Module.Limit Speed=true, Velocity Limit=1000)
  ]
  particleUpdateScript [
   &ParticleState()
   &GravityForce(Gravity=&RandomRangeVector(Minimum=<0 0 -377.8992>, Maximum=<1 1 -898.1059>))
   &SolveForcesAndVelocity(Clamp Velocity=true, Limit Acceleration=false, Speed Limit=54.9483)
   &ScaleSpriteSize(Scale Factor=&Multiply_Vector2(
     A=<1 1>
     B=&Vector2DFromFloat(Value=&FloatFromCurve(FloatCurve=DataInterface() {propertyValues {Curve {keys [
           {interpMode=RCIM_Cubic, tangentMode=RCTM_Break, time=-0.0046, value=0}
           {arriveTangent=0.0019, interpMode=RCIM_Cubic, leaveTangent=0.0019, tangentMode=RCTM_Break, time=0.1, value=3}
           {interpMode=RCIM_Cubic, tangentMode=RCTM_Break, time=1.0011, value=0.002}
          ]}}}))
    ))
   /SetVariables(Particles.Color=&MakeLinearColorFromVectorAndFloat(
     "Vector (RGB)"=&MakeVectorFromLinearColorRGB(Color=&ColorFromCurve(ColorCurve=DataInterface() {propertyValues {
         AlphaCurve {keys [<at 0 0 RCIM_Cubic>,<at 0.9977 0 RCIM_Cubic>]}
         BlueCurve {keys [<at 0 0.254 RCIM_Cubic>,<at 1 0.1605>]}
         GreenCurve {keys [<at 0 0.4032 RCIM_Cubic>,<at 1 0.2882>]}
         RedCurve {keys [<at 0 1 RCIM_Cubic>,<at 1 0.75>]}
        }}))
     "Float (Alpha)"=&FloatFromCurve(
      FloatCurve=DataInterface() {propertyValues {Curve {keys [<at 0 0>,<at 0.1936 0.8315>,<at 0.4019 0.1015>,<at 0.5589 0.6516>,<at 0.7294 0.0749>,<at 0.8673 0.41>,<at 1.0011 0.002>]}}}
      Scale Curve=0.2
     )
    )) {isSetParameters=true}
  ]
  rendererClasses [&NiagaraSpriteRendererProperties]
  renderers [{rendererClass=&NiagaraSpriteRendererProperties}]
 }
 Jet {
  emitterUpdateScript [
   &EmitterState(
    Inactive Response=::NewEnumerator1
    Life Cycle Mode=::NewEnumerator1
    Loop Behavior=::NewEnumerator1
    Loop Duration Mode=::NewEnumerator1
    UseLoopDelay=false
   )
   &SpawnBurst_Instantaneous(Spawn Count=1, Use Spawn Probability=false)
  ]
  particleSpawnScript [
   /Niagara/Modules/Spawn/Initialization/V2/InitializeParticle(
    Color Mode=::NewEnumerator0
    Lifetime=1000000000
    Mesh Scale=<5.5105 10 10>
    Mesh Scale Mode=::NewEnumerator4
    UsePositionOffset=false
    Write Material Random=true
   )
   &InitialMeshOrientation(Orientation Vector=<0 0 -1>, Use Orientation Vector=true, Use Rotation Vector=false)
  ]
  particleUpdateScript [
   &ParticleState()
   &ScaleMeshSize(Scale Factor=&RandomRangeVector(Minimum=<0.95 1 1>, Maximum=<1.03 1 1>, Evaluation Type=::NewEnumerator1))
   &ScaleColor(
    Scale RGB=&VectorFromFloat(Value=&RandomRangeFloat(Minimum=0.6, Maximum=1, Evaluation Type=::NewEnumerator1))
    ScaleA=true
    ScaleRGB=true
   )
   /SetVariables(Particles.TAG:NiagaraInt32=&MakeCustomIntFromBool(Bool=$User.bIsFrontEndPreview, True INT=1, False INT=0)) {isSetParameters=true, module=SetVariables_#1}
  ]
  rendererClasses [&NiagaraSpriteRendererProperties, &NiagaraMeshRendererProperties]
  renderers [
   {rendererClass=&NiagaraSpriteRendererProperties}
   {rendererClass=&NiagaraMeshRendererProperties, rendererIndex=1}
   {rendererClass=&NiagaraMeshRendererProperties, rendererIndex=2}
  ]
 }
 Jet001 {
  emitterUpdateScript [
   &EmitterState(
    Inactive Response=::NewEnumerator1
    Life Cycle Mode=::NewEnumerator1
    Loop Behavior=::NewEnumerator1
    Loop Duration Mode=::NewEnumerator1
    UseLoopDelay=false
   )
   &SpawnBurst_Instantaneous(Spawn Count=1, Use Spawn Probability=false)
  ]
  particleSpawnScript [
   &InitializeParticle(
    Color Mode=::NewEnumerator0
    Lifetime=1000000000
    Mesh Scale=<5.5105 10 10>
    Mesh Scale Mode=::NewEnumerator4
    UsePositionOffset=false
    Write Material Random=true
   )
   &InitialMeshOrientation(Orientation Vector=<0 0 -1>, Use Orientation Vector=true, Use Rotation Vector=false)
  ]
  particleUpdateScript [
   &ParticleState()
   &ScaleMeshSize(Scale Factor=&RandomRangeVector(Minimum=<0.95 1 1>, Maximum=<1.03 1 1>, Evaluation Type=::NewEnumerator1))
   &ScaleColor(
    Scale Alpha=&MakeCustomFloatFromBool(Bool=$User.bIsFrontEndPreview, True Float=0.35, False Float=1)
    Scale RGB=&VectorFromFloat(Value=&RandomRangeFloat(Minimum=0.4636, Maximum=1, Evaluation Type=::NewEnumerator1))
    ScaleA=true
    ScaleRGB=true
   )
   /SetVariables(Particles.TAG:NiagaraInt32=&MakeCustomIntFromBool(Bool=$User.bIsFrontEndPreview, True INT=1, False INT=0)) {isSetParameters=true, module=SetVariables_#1}
  ]
  rendererClasses [&NiagaraSpriteRendererProperties, &NiagaraMeshRendererProperties]
  renderers [
   {rendererClass=&NiagaraSpriteRendererProperties}
   {rendererClass=&NiagaraMeshRendererProperties, rendererIndex=1}
   {rendererClass=&NiagaraMeshRendererProperties, rendererIndex=2}
  ]
 }
 Jet02 {
  emitterUpdateScript [
   &EmitterState(
    Inactive Response=::NewEnumerator1
    Life Cycle Mode=::NewEnumerator1
    Loop Behavior=::NewEnumerator1
    Loop Duration Mode=::NewEnumerator1
    UseLoopDelay=false
   )
   &SpawnBurst_Instantaneous(Spawn Count=1, Use Spawn Probability=false)
  ]
  particleSpawnScript [
   &InitializeParticle(
    Color Mode=::NewEnumerator0
    Lifetime=1000000000
    Mesh Scale=<15 20 20>
    Mesh Scale Mode=::NewEnumerator4
    UsePositionOffset=false
    Write Material Random=true
   )
   &InitialMeshOrientation(Orientation Vector=<0 0 -1>, Use Orientation Vector=true, Use Rotation Vector=false)
  ]
  particleUpdateScript [
   &ParticleState()
   &ScaleMeshSize(Scale Factor=&RandomRangeVector(Minimum=<0.95 1 1>, Maximum=<1.03 1 1>, Evaluation Type=::NewEnumerator1))
   &ScaleColor(
    Scale RGB=&VectorFromFloat(Value=&RandomRangeFloat(Minimum=0.5, Maximum=1, Evaluation Type=::NewEnumerator1))
    ScaleA=true
    ScaleRGB=true
   )
   /SetVariables(Particles.TAG:NiagaraInt32=&MakeCustomIntFromBool(Bool=$User.bIsFrontEndPreview, True INT=1, False INT=0)) {isSetParameters=true}
  ]
  rendererClasses [&NiagaraSpriteRendererProperties, &NiagaraMeshRendererProperties]
  renderers [
   {rendererClass=&NiagaraSpriteRendererProperties}
   {rendererClass=&NiagaraMeshRendererProperties, rendererIndex=1}
   {rendererClass=&NiagaraMeshRendererProperties, rendererIndex=2}
  ]
 }
 ParticleSpotLight {
  emitterUpdateScript [
   &EmitterState(
    Inactive Response=::NewEnumerator1
    Life Cycle Mode=::NewEnumerator1
    Loop Behavior=::NewEnumerator1
    Loop Duration Mode=::NewEnumerator1
    UseLoopDelay=false
   )
   &SpawnBurst_Instantaneous(Spawn Count=1, Spawn Time=0.01, Use Spawn Probability=false)
  ]
  particleSpawnScript [&InitializeParticle(Color Mode=::NewEnumerator0, Lifetime=10000000000, UsePositionOffset=false, Write Material Random=true)]
  particleUpdateScript [
   &ParticleState()
   /SetVariables(Particles.SpotLightColor:LinearColor=<rgba 1 0.5555 0.5 1>) {isSetParameters=true}
   /SetVariables() {isSetParameters=true}
  ]
  rendererClasses [&NiagaraComponentRendererProperties]
  renderers [{rendererClass=&NiagaraComponentRendererProperties, rendererIndex=0}]
 }
}
summary {
 systemName=NS_Emote_SaucerPart
 emitters [
  {emitterName=ParticleSpotLight, rendererClasses [&NiagaraComponentRendererProperties]}
  {emitterName=FN_SparkleDust_Constant, rendererClasses [&NiagaraSpriteRendererProperties]}
  {emitterName=Jet, rendererClasses [&NiagaraSpriteRendererProperties, &NiagaraMeshRendererProperties]}
  {emitterName=Jet02, rendererClasses [&NiagaraSpriteRendererProperties, &NiagaraMeshRendererProperties]}
  {emitterName=Jet001, rendererClasses [&NiagaraSpriteRendererProperties, &NiagaraMeshRendererProperties]}
 ]
 userVariables [
  User.bIsFrontEndPreview {type=NiagaraBool}
  User.SkeletalMesh {
   type=NiagaraDataInterfaceSkeletalMesh
   defaultValue {
    struct=NiagaraExt_VariableValue_DataInterface
    value {dataInterface=&self:NiagaraDataInterfaceSkeletalMesh_1}
   }
  }
 ]
}
